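# Editor utility widget that bakes a Niagara FLIP simulation into normal, motion-vector and thickness flipbook textures.
# Its buttons spawn a FlipSimBaker actor, open documentation, or run the bake, with settings in a details view.
# Baking captures ResX×ResY timed frames through the actor into render targets, then tiles them into three atlases using per-channel canvas materials.
# Atlases are redrawn at ScaleToSize, then written into existing textures with compression and sRGB set.
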EditorUtilityWidgetBlueprint /Game/Effects/FlipSimulation/BakingTool/FlipSimBakerUI
graphs [EventGraph {nodes [
   "On Clicked (EditorUtilityButton_1)"() {is=title K2Node_ComponentBoundEvent}
   > n1 = Sequence!(execute) -> (then_0, then_1, then_2:Exec, then_3:Exec)
   n2 = Get Target Flip Sim Baker() -> (Target Flip Sim Baker:Flip Sim Baker Object Reference)
   > Set GridSize!(execute, GridSize:Vector 2D Structure=Get GridSize() -> (GridSize:Vector 2D Structure), self:Flip Sim Baker Object Reference=@n2) -> (then, Output_Get:Vector 2D Structure)
   > Set GridPivot!(execute, GridPivot:Vector 2D Structure=Get GridPivot() -> (GridPivot:Vector 2D Structure), self:Flip Sim Baker Object Reference=@n2) -> (then, Output_Get:Vector 2D Structure)
   > Initial!(
    execute
    self:Flip Sim Baker Object Reference=@n2
    TargetNS:Niagara System Object Reference=Get TargetNS() -> (TargetNS:Niagara System Object Reference)
   ) -> (then)
   > Set Actor Rotation!(execute, self:Actor Object Reference=@n2, NewRotation:Rotator="0,90,0", bTeleportPhysics:bool=false) -> (then, ReturnValue:bool) {is=title K2Node_CallFunction}
   n3 = Get Target Flip Sim Baker()
   n4 = BreakTransform(InTransform=Get Actor Transform(self=@n3))
   SetLevelViewportCameraInfo!(
    execute=@n1.then_1
    self:Unreal Editor Subsystem Object Reference=Get UnrealEditorSubsystem()
    CameraLocation:Vector=@n4
    CameraRotation:Rotator=@n4.Rotation
   ) -> (then)
   n5 = Get ScaleToSize() -> (ScaleToSize:Vector 2D Structure)
   n6 = BreakVector2D(InVec=@n5)
   > n7 = CreateRenderTarget2D!(
    execute
    WorldContextObject:Object Reference
    Width:int=Truncate(A=@n6)
    Height:int=Truncate(A=@n6.Y)
    Format:ETextureRenderTargetFormat Enum=RTF_RGBA16f
    ClearColor:Linear Color Structure=(R=0,G=0,B=0,A=1)
    bAutoGenerateMipMaps:bool=false
    bSupportUAVs:bool=false
   ) -> (then, ReturnValue:Texture Render Target 2D Object Reference)
   > Set FinalRT_N(FinalRT_N:Texture Render Target 2D Object Reference=@n7) -> (Output_Get:Texture Render Target 2D Object Reference)
   > Set AddressX!(execute, AddressX:TextureAddress Enum=TA_Clamp, self:Texture Render Target 2D Object Reference=@n7) -> (then, Output_Get:TextureAddress Enum=TA_Wrap)
   > Set AddressY!(execute, AddressY:TextureAddress Enum=TA_Clamp, self:Texture Render Target 2D Object Reference=@n7) -> (then, Output_Get:TextureAddress Enum=TA_Wrap)
   n8 = BreakVector2D(InVec=@n5)
   > n9 = CreateRenderTarget2D(
    Width=Truncate(A=@n8)
    Height=Truncate(A=@n8.Y)
    Format=RTF_RGBA16f
    ClearColor=(R=0,G=0,B=0,A=1)
    bAutoGenerateMipMaps=false
    bSupportUAVs=false
   )
   > Set AddressX(AddressX=TA_Clamp, self=@n9) -> (Output_Get=TA_Wrap)
   > Set AddressY(AddressY=TA_Clamp, self=@n9) -> (Output_Get=TA_Wrap)
   > Set FinalRT_MV(FinalRT_MV:Texture Render Target 2D Object Reference=@n9) -> (Output_Get:Texture Render Target 2D Object Reference)
   n10 = BreakVector2D(InVec=@n5)
   > n11 = CreateRenderTarget2D(
    Width=Truncate(A=@n10)
    Height=Truncate(A=@n10.Y)
    Format=RTF_RGBA16f
    ClearColor=(R=0,G=0,B=0,A=1)
    bAutoGenerateMipMaps=false
    bSupportUAVs=false
   )
   > Set AddressX(AddressX=TA_Clamp, self=@n11) -> (Output_Get=TA_Wrap)
   > Set AddressY(AddressY=TA_Clamp, self=@n11) -> (Output_Get=TA_Wrap)
   > Set FinalRT_T(FinalRT_T:Texture Render Target 2D Object Reference=@n11) -> (Output_Get:Texture Render Target 2D Object Reference)
   n12 = Get RTs() -> (RTs:Array of Texture Render Target 2D Object References)
   n13 = For Each Loop(Exec=@n1.then_2, Array:Array of Texture Render Target 2D Object References=@n12) -> (Array Element:Texture Render Target 2D Object Reference)
   Clear(execute=@n13.Completed, TargetArray:Array of Texture Render Target 2D Object References=@n12)
   > Set Counter(Counter:int=0) -> (Output_Get:int=0)
   ReleaseRenderTarget2D!(execute=@n13.LoopBody, TextureRenderTarget:Texture Render Target 2D Object Reference=@n13) -> (then)
   Delay!(execute=@n1.then_3, WorldContextObject:Object Reference, Duration=0.2) -> (then)
   > n14 = Reset(self:Flip Sim Baker Object Reference=@n3)
   n15 = Get ResY() -> (ResY:int=0)
   n16 = Get ResX() -> (ResX:int=0)
   Delay(
    execute=[@n84.then, @n14.then, @n87.Completed]
    Duration=float / float(A=Get CaptureTime() -> (CaptureTime:double), B=float * float(A:int=@n16, B:int=@n15) {is=title K2Node_PromotableOperator})
   )
   > Increment Int(Value=Get Counter() -> (Counter:int=0))
   > n17 = Branch(Condition="integer >= integer"(A=Get Counter() -> (Counter=0), B=int * int(A=@n16, B=@n15)))
   > Delay(Duration=1)
   n18 = BreakVector2D(InVec=Get AssembleRTSize() -> (AssembleRTSize:Vector 2D Structure))
   n19 = Truncate(A=@n18)
   n20 = Truncate(A=@n18.Y)
   > n21 = CreateRenderTarget2D(Width=@n19, Height=@n20, Format=RTF_RGBA16f, ClearColor=(R=0,G=0,B=0,A=1), bAutoGenerateMipMaps=false, bSupportUAVs=false)
   > n22 = Set AssembleRT_N(AssembleRT_N:Texture Render Target 2D Object Reference=@n21) -> (Output_Get:Texture Render Target 2D Object Reference)
   > Set AddressX(AddressX=TA_Clamp, self=@n22) -> (Output_Get=TA_Wrap)
   > Set AddressY(AddressY=TA_Clamp, self=@n22) -> (Output_Get=TA_Wrap)
   > n23 = CreateRenderTarget2D(Width=@n19, Height=@n20, Format=RTF_RGBA16f, ClearColor=(R=0,G=0,B=0,A=1), bAutoGenerateMipMaps=false, bSupportUAVs=false)
   > n24 = Set AssembleRT_MV(AssembleRT_MV:Texture Render Target 2D Object Reference=@n23) -> (Output_Get:Texture Render Target 2D Object Reference)
   > Set AddressX(AddressX=TA_Clamp, self=@n24) -> (Output_Get=TA_Wrap)
   > Set AddressY(AddressY=TA_Clamp, self=@n24) -> (Output_Get=TA_Wrap)
   > n25 = CreateRenderTarget2D(Width=@n19, Height=@n20, Format=RTF_RGBA16f, ClearColor=(R=0,G=0,B=0,A=1), bAutoGenerateMipMaps=false, bSupportUAVs=false)
   > n26 = Set AssembleRT_T(AssembleRT_T:Texture Render Target 2D Object Reference=@n25) -> (Output_Get:Texture Render Target 2D Object Reference)
   > Set AddressX(AddressX=TA_Clamp, self=@n26) -> (Output_Get=TA_Wrap)
   > Set AddressY(AddressY=TA_Clamp, self=@n26) -> (Output_Get=TA_Wrap)
   > n27 = Sequence()
   n28 = Get AssembleRT_N() -> (AssembleRT_N:Texture Render Target 2D Object Reference)
   > Is Valid(InputObject=@n28)
   > n29 = CreateDynamicMaterialInstance(Parent=&/WriteRTMat_N, OptionalName=None, CreationFlags=None)
   > n30 = Set Mat(Mat:Material Instance Dynamic Object Reference=@n29) -> (Output_Get:Material Instance Dynamic Object Reference)
   > n31 = BeginDrawCanvasToRenderTarget!(execute, WorldContextObject:Object Reference, TextureRenderTarget:Texture Render Target 2D Object Reference=@n28) -> (then, Canvas:Canvas Object Reference, Size:Vector 2D Structure, Context:Draw To Render Target Context Structure)
   > n32 = For Loop(LastIndex=int - int(A=Get ResY() -> (ResY=0)))
   > n33 = For Loop(LastIndex=int - int(A=Get ResX() -> (ResX=0)))
   n34 = int * int(A=@n32, B=Get ResX() -> (ResX=0))
   > SetTextureParameterValue(
    self=Get Mat() -> (Mat:Material Instance Dynamic Object Reference)
    ParameterName=RT
    Value="Get (a copy)"(Array:Array of Texture Render Target 2D Object References=Get RTs(), Dimension 1=int + int(A=@n34, B=@n33)) -> (Output:Texture Render Target 2D Object Reference)
   )
   n35 = Get ResY() -> (ResY=0)
   n36 = Get ResX() -> (ResX=0)
   n37 = vector2d / vector2d(A:Vector 2D Structure=@n31.Size, B:Vector 2D Structure=MakeVector2D(X="To Float (Integer)"(InInt=@n36), Y="To Float (Integer)"(InInt=@n35))) -> (ReturnValue:Vector 2D Structure) {is=title K2Node_PromotableOperator}
   n38 = BreakVector2D(InVec=@n37)
   n39 = "To Float (Integer)"(InInt=@n33)
   n40 = "To Float (Integer)"(InInt=@n32)
   > Draw Material!(
    execute
    self:Canvas Object Reference=@n31
    RenderMaterial:Material Interface Object Reference=@n30
    ScreenPosition:Vector 2D Structure=MakeVector2D(X=float * float(A:double=@n39, B:double=@n38), Y=float * float(A=@n40, B=@n38.Y))
    ScreenSize:Vector 2D Structure=@n37
    CoordinatePosition:Vector 2D Structure
    CoordinateSize:Vector 2D Structure=(X=1,Y=1)
    Rotation:float=0
    PivotPoint:Vector 2D Structure=(X=0.5,Y=0.5)
   ) -> (then)
   n41 = Get AssembleRT_MV() -> (AssembleRT_MV:Texture Render Target 2D Object Reference)
   Is Valid(exec=@n27.then_1, InputObject=@n41)
   > n42 = CreateDynamicMaterialInstance(Parent=&/WriteRTMat_MV, OptionalName=None, CreationFlags=None)
   > Set Mat(Mat=@n42)
   > n43 = BeginDrawCanvasToRenderTarget(TextureRenderTarget=@n41)
   > n44 = For Loop(LastIndex=int - int(A=Get ResY() -> (ResY=0)))
   > n45 = For Loop(LastIndex=int - int(A=Get ResX() -> (ResX=0)))
   n46 = int * int(A=@n44, B=Get ResX() -> (ResX=0))
   n47 = Get Mat()
   > SetTextureParameterValue(self=@n47, ParameterName=RT, Value="Get (a copy)"(Array=Get RTs(), Dimension 1=int + int(A=@n46, B=@n45)))
   n48 = Get ResY() -> (ResY=0)
   n49 = Get ResX() -> (ResX=0)
   n50 = vector2d / vector2d(A=@n43.Size, B=MakeVector2D(X="To Float (Integer)"(InInt=@n49), Y="To Float (Integer)"(InInt=@n48)))
   n51 = BreakVector2D(InVec=@n50)
   n52 = "To Float (Integer)"(InInt=@n45)
   n53 = "To Float (Integer)"(InInt=@n44)
   > Draw Material(
    self=@n43
    RenderMaterial=@n47
    ScreenPosition=MakeVector2D(X=float * float(A=@n52, B=@n51), Y=float * float(A=@n53, B=@n51.Y))
    ScreenSize=@n50
    CoordinateSize=(X=1,Y=1)
    Rotation=0
    PivotPoint=(X=0.5,Y=0.5)
   )
   n54 = Get AssembleRT_T() -> (AssembleRT_T:Texture Render Target 2D Object Reference)
   Is Valid(exec=@n27.then_2, InputObject=@n54)
   > n55 = CreateDynamicMaterialInstance(Parent=&/WriteRTMat_T, OptionalName=None, CreationFlags=None)
   > Set Mat(Mat=@n55)
   > n56 = BeginDrawCanvasToRenderTarget(TextureRenderTarget=@n54)
   > n57 = For Loop(LastIndex=int - int(A=Get ResY() -> (ResY=0)))
   > n58 = For Loop(LastIndex=int - int(A=Get ResX() -> (ResX=0)))
   n59 = int * int(A=@n57, B=Get ResX() -> (ResX=0))
   n60 = Get Mat()
   > SetTextureParameterValue(self=@n60, ParameterName=RT, Value="Get (a copy)"(Array=Get RTs(), Dimension 1=int + int(A=@n59, B=@n58)))
   n61 = Get ResY() -> (ResY=0)
   n62 = Get ResX() -> (ResX=0)
   n63 = vector2d / vector2d(A=@n56.Size, B=MakeVector2D(X="To Float (Integer)"(InInt=@n62), Y="To Float (Integer)"(InInt=@n61)))
   n64 = BreakVector2D(InVec=@n63)
   n65 = "To Float (Integer)"(InInt=@n57)
   n66 = "To Float (Integer)"(InInt=@n58)
   > Draw Material(
    self=@n56
    RenderMaterial=@n60
    ScreenPosition=MakeVector2D(X=float * float(A=@n66, B=@n64), Y=float * float(A=@n65, B=@n64.Y))
    ScreenSize=@n63
    CoordinateSize=(X=1,Y=1)
    Rotation=0
    PivotPoint=(X=0.5,Y=0.5)
   )
   n67 = CreateDynamicMaterialInstance(execute=@n27.then_3, Parent=&/WriteRTMat_Final, OptionalName=None, CreationFlags=None)
   > n68 = Set Mat_Final(Mat_Final:Material Instance Dynamic Object Reference=@n67) -> (Output_Get:Material Instance Dynamic Object Reference)
   > SetTextureParameterValue(self=@n68, ParameterName=RT, Value=Get AssembleRT_N())
   > n69 = BeginDrawCanvasToRenderTarget(TextureRenderTarget=Get FinalRT_N() -> (FinalRT_N:Texture Render Target 2D Object Reference))
   > Draw Material(
    self=@n69
    RenderMaterial=Get Mat_Final() -> (Mat_Final:Material Instance Dynamic Object Reference)
    ScreenSize=@n69.Size
    CoordinateSize=(X=1,Y=1)
    Rotation=0
    PivotPoint=(X=0.5,Y=0.5)
   )
   > n70 = BeginDrawCanvasToRenderTarget(TextureRenderTarget=Get FinalRT_MV() -> (FinalRT_MV:Texture Render Target 2D Object Reference))
   n71 = Get Mat_Final()
   > SetTextureParameterValue(self=@n71, ParameterName=RT, Value=Get AssembleRT_MV())
   > Draw Material(self=@n70, RenderMaterial=@n71, ScreenSize=@n70.Size, CoordinateSize=(X=1,Y=1), Rotation=0, PivotPoint=(X=0.5,Y=0.5))
   > n72 = CreateDynamicMaterialInstance(Parent=&/WriteRTMat_Final_Alpha, OptionalName=None, CreationFlags=None)
   > Set Mat_Final(Mat_Final=@n72)
   > n73 = BeginDrawCanvasToRenderTarget(TextureRenderTarget=Get FinalRT_T() -> (FinalRT_T:Texture Render Target 2D Object Reference))
   n74 = Get Mat_Final()
   > SetTextureParameterValue(self=@n74, ParameterName=RT, Value=Get AssembleRT_T())
   > Draw Material(self=@n73, RenderMaterial=@n74, ScreenSize=@n73.Size, CoordinateSize=(X=1,Y=1), Rotation=0, PivotPoint=(X=0.5,Y=0.5))
   > n75 = Sequence()
   n76 = Get BakeToTexture_T() -> (BakeToTexture_T:Texture 2D Object Reference)
   > Is Valid(InputObject=@n76)
   > ConvertRenderTargetToTexture2DEditorOnly(RenderTarget=Get FinalRT_T(), Texture=@n76)
   n77 = Get BakeToTexture_T()
   > SetEditorProperty(Object=@n77, PropertyName=CompressionNoAlpha, PropertyValue=OR Boolean(A=false, B=false))
   > SetEditorProperty(
    Object=@n77
    PropertyName=CompressionSettings
    PropertyValue:TextureCompressionSettings Enum=Get CD() -> (CD:TextureCompressionSettings Enum)
   )
   > SetEditorProperty(Object=@n77, PropertyName=SRGB, PropertyValue:bool=OR Boolean(A=true, B=false))
   n78 = Get BakeToTexture_N() -> (BakeToTexture_N:Texture 2D Object Reference)
   Is Valid(exec=@n75.then_1, InputObject=@n78)
   > ConvertRenderTargetToTexture2DEditorOnly(RenderTarget=Get FinalRT_N(), Texture=@n78)
   n79 = Get BakeToTexture_N()
   > SetEditorProperty(
    Object=@n79
    PropertyName=CompressionSettings
    PropertyValue:TextureCompressionSettings Enum=Get CN() -> (CN:TextureCompressionSettings Enum)
   )
   > SetEditorProperty(Object=@n79, PropertyName=SRGB, PropertyValue:bool=OR Boolean(A=false, B=false))
   n80 = Get BakeToTexture_MV() -> (BakeToTexture_MV:Texture 2D Object Reference)
   Is Valid(exec=@n75.then_2, InputObject=@n80)
   > ConvertRenderTargetToTexture2DEditorOnly(RenderTarget=Get FinalRT_MV(), Texture=@n80)
   n81 = Get BakeToTexture_MV()
   > SetEditorProperty(Object=@n81, PropertyName=CompressionSettings, PropertyValue:TextureCompressionSettings Enum=Get CN())
   > SetEditorProperty(Object=@n81, PropertyName=SRGB, PropertyValue:bool=OR Boolean(A=false, B=false))
   ReleaseRenderTarget2D(execute=@n75.then_3, TextureRenderTarget=Get AssembleRT_T())
   > ReleaseRenderTarget2D(TextureRenderTarget=Get AssembleRT_N())
   > ReleaseRenderTarget2D(TextureRenderTarget=Get AssembleRT_MV())
   > ReleaseRenderTarget2D(TextureRenderTarget=Get FinalRT_T())
   > ReleaseRenderTarget2D(TextureRenderTarget=Get FinalRT_N())
   > ReleaseRenderTarget2D(TextureRenderTarget=Get FinalRT_MV())
   n82 = BreakVector2D(InVec=Get GridSize())
   n83 = CreateRenderTarget2D(
    execute=@n17.else
    WorldContextObject=Get Target Flip Sim Baker()
    Width=Truncate(A=@n82)
    Height=Truncate(A=@n82.Y)
    Format=RTF_RGBA16f
    ClearColor=(R=0,G=0,B=0,A=1)
    bAutoGenerateMipMaps=false
    bSupportUAVs=false
   )
   > Add(TargetArray:Array of Texture Render Target 2D Object References=Get RTs(), NewItem:"Texture Render Target 2D Object Reference (by ref)"=@n83) {is=title K2Node_CallArrayFunction}
   > n84 = Capture!(execute, self:Flip Sim Baker Object Reference=Get Target Flip Sim Baker(), NewParam:Texture Render Target 2D Object Reference=@n83) -> (then)
   "On Clicked (EditorUtilityButton)"() {is=title K2Node_ComponentBoundEvent}
   > n85 = SpawnActor Flip Sim Baker!(
    execute:Exec
    WorldContextObject:Object Reference
    Class:Actor Class Reference=&/FlipSimBaker.FlipSimBaker_C
    SpawnTransform:Transform=MakeTransform(Location="20000,20000,20000", Rotation="0,90,0")
    CollisionHandlingOverride:ESpawnActorCollisionHandlingMethod Enum=AlwaysSpawn
    TransformScaleMethod:ESpawnActorScaleMethod Enum=MultiplyWithRoot
    Owner:Actor Object Reference
    Instigator:Pawn Object Reference
   ) -> (then:Exec, ReturnValue:Flip Sim Baker Object Reference) {is=title K2Node_SpawnActorFromClass}
   > Set Target Flip Sim Baker(Target Flip Sim Baker:Flip Sim Baker Object Reference=@n85) -> (Output_Get:Flip Sim Baker Object Reference)
   "On Clicked (EditorUtilityButton_2)"() {is=title K2Node_ComponentBoundEvent}
   > LaunchURL!(execute, URL:String="https://docs.google.com/document/d/1dd9iQST_62M--BkeRm2AiB07eQNEFKabU7qpyJ_OJVI/edit?usp=sharing") -> (then)
   Event Construct()
   > SetObject(self=Get DetailsView_190() -> (DetailsView_190:Details View Object Reference), NewObject=Self-Reference())
   > n86 = GetActorOfClass!(
    execute
    WorldContextObject:Object Reference
    ActorClass:Actor Class Reference=&FlipSimBaker_C
   ) -> (then, ReturnValue:Flip Sim Baker Object Reference)
   > Set Target Flip Sim Baker(Target Flip Sim Baker=@n86)
   n87 = For Loop(LastIndex=int - int(A=Get Counter() -> (Counter=0)))
   n88 = BreakVector2D(InVec=Get GridSize())
   > n89 = CreateRenderTarget2D(
    WorldContextObject=Get Target Flip Sim Baker()
    Width=Truncate(A=@n88)
    Height=Truncate(A=@n88.Y)
    Format=RTF_RGBA16f
    ClearColor=(R=0,G=0,B=0,A=1)
    bAutoGenerateMipMaps=false
    bSupportUAVs=false
   )
   > Add(TargetArray=Get RTs(), NewItem=@n89)
   Event PreConstruct()
   Event Tick()
   "Get (a copy)"(Array=Get RTs(), Dimension 1=int - int(A=Get Counter() -> (Counter=0)))
   Get FinalRT_N()
  ]}]
variables [
 AssembleRT_MV {type=TextureRenderTarget2D, default=null}
 AssembleRT_N {type=TextureRenderTarget2D, default=null}
 AssembleRT_T {type=TextureRenderTarget2D, default=null}
 AssembleRTSize {type=Vector2D, default=<2048 2048>}
 BakeToTexture_MV {type=Texture2D, default=&/T_BakedMotionVector}
 BakeToTexture_N {type=Texture2D, default=&/T_BakedNormal}
 BakeToTexture_T {type=Texture2D, default=&/T_BakedThickness}
 CaptureTime {type=number, default=0.7}
 CD {type=TextureCompressionSettings, default=TC_Default}
 CN {type=TextureCompressionSettings, default=TC_Normalmap}
 Counter {type=integer}
 FinalRT_MV {type=TextureRenderTarget2D, default=null}
 FinalRT_N {type=TextureRenderTarget2D, default=null}
 FinalRT_T {type=TextureRenderTarget2D, default=null}
 GridPivot {type=Vector2D, default=<0.5 0.5>}
 GridSize {type=Vector2D, default=<512 512>}
 Mat {type=MaterialInstanceDynamic}
 Mat_Final {type=MaterialInstanceDynamic}
 ResX {type=integer, default=4}
 ResY {type=integer, default=4}
 RTs {type="array<TextureRenderTarget2D>", default []}
 ScaleToSize {type=Vector2D, default=<1024 1024>}
 Target Flip Sim Baker {type=FlipSimBaker_C, default=null}
 TargetNS {type=NiagaraSystem, default=null, class=~, dependencies=~, emitters=~, summary=~}
]
widgets {
 info {rootWidgetClass=&CanvasPanel, widgetCount=8}
 widgets [
  CanvasPanel_138 {widgetClassPath=&CanvasPanel, widget=CanvasPanel_138}
  DetailsView_190 {
   widgetClassPath=&DetailsView
   parent=CanvasPanel_138
   widget=DetailsView_190
   properties {categoriesToShow [Default]}
   slot {
    class=CanvasPanelSlot
    properties {layoutData {alignment=<0 0>, anchors {maximum=<0 0>, minimum=<0 0>}, offsets=<ltrb 0 0 694.5946 475.7695>}}
   }
  }
  EditorUtilityButton_1 {
   widgetClassPath=&EditorUtilityButton
   parent=CanvasPanel_138
   widget=EditorUtilityButton_1
   slot {class=CanvasPanelSlot, properties {layoutData {alignment=<0 0>, anchors {maximum=<0 0>, minimum=<0 0>}, offsets=<ltrb 8 488 100 40>}}}
  }
  TextBlock_1 {
   widgetClassPath=&TextBlock
   parent=EditorUtilityButton_1
   widget=TextBlock_1
   properties {
    justification=Center
    text=Bake
    font {
     bForceMonospaced=false
     bMaterialIsStencil=false
     fontMaterial=null
     fontObject=&Roboto
     letterSpacing=0
     monospacedWidth=1
     size=7.5
     skewAmount=0
     typefaceFontName=Bold
     outlineSettings {
      bApplyOutlineToDropShadows=false
      bMiteredCorners=false
      bSeparateFillAlpha=false
      outlineMaterial=null
      outlineSize=0
      outlineColor=<rgba 0 0 0 1>
     }
    }
   }
   slot {class=ButtonSlot}
  }
  EditorUtilityButton {
   widgetClassPath=&EditorUtilityButton
   parent=CanvasPanel_138
   widget=EditorUtilityButton
   slot {class=CanvasPanelSlot, properties {layoutData {alignment=<0 0>, anchors {maximum=<0 0>, minimum=<0 0>}, offsets=<ltrb 124 488 100 40>}}}
  }
  TextBlock {
   widgetClassPath=&TextBlock
   parent=EditorUtilityButton
   widget=TextBlock
   properties {
    justification=Center
    text="Create Baker Actor "
    font {
     bForceMonospaced=false
     bMaterialIsStencil=false
     fontMaterial=null
     fontObject=&Roboto
     letterSpacing=0
     monospacedWidth=1
     size=7.5
     skewAmount=0
     typefaceFontName=Bold
     outlineSettings {
      bApplyOutlineToDropShadows=false
      bMiteredCorners=false
      bSeparateFillAlpha=false
      outlineMaterial=null
      outlineSize=0
      outlineColor=<rgba 0 0 0 1>
     }
    }
   }
   slot {class=ButtonSlot}
  }
  EditorUtilityButton_2 {
   widgetClassPath=&EditorUtilityButton
   parent=CanvasPanel_138
   widget=EditorUtilityButton_2
   slot {class=CanvasPanelSlot, properties {layoutData {alignment=<0 0>, anchors {maximum=<0 0>, minimum=<0 0>}, offsets=<ltrb 244 488 100 40>}}}
  }
  TextBlock_2 {
   widgetClassPath=&TextBlock
   parent=EditorUtilityButton_2
   widget=TextBlock_2
   properties {
    justification=Center
    text=Open Doc
    font {
     bForceMonospaced=false
     bMaterialIsStencil=false
     fontMaterial=null
     fontObject=&Roboto
     letterSpacing=0
     monospacedWidth=1
     size=7.5
     skewAmount=0
     typefaceFontName=Bold
     outlineSettings {
      bApplyOutlineToDropShadows=false
      bMiteredCorners=false
      bSeparateFillAlpha=false
      outlineMaterial=null
      outlineSize=0
      outlineColor=<rgba 0 0 0 1>
     }
    }
   }
   slot {class=ButtonSlot}
  }
 ]
}
